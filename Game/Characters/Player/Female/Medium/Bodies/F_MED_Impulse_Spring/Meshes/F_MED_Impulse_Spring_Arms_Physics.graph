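# Physics asset for the arms mesh of the female medium player body Impulse Spring, setting up dangling straps on both upper arms.
# Each arm has two short capsule chains, strap A and strap B, each linking two bones in series from the upper arm twist bone.
# Constraints keep twist almost locked while allowing moderate swing, with the root joint of each B strap tighter on one swing axis.
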PhysicsAsset /Game/Characters/Player/Female/Medium/Bodies/F_MED_Impulse_Spring/Meshes/F_MED_Impulse_Spring_Arms_Physics
bodies {
 dyn_straps_l_a02 [{
   shapeType=Capsule
   length=1.6944
   radius=1.25
   shapeName=dyn_straps_r_a01_capsule
   center=<1.4897 -0.5003 0.3675>
   rotation=<rot 90 0 -10.0004>
  }]
 dyn_straps_l_a03 [{shapeType=Capsule, length=1.6944, radius=1.25, shapeName=dyn_straps_r_a01_capsule, center=<1.3961 0.0116 0>, rotation=<rot 90 0 0>}]
 dyn_straps_l_b02 [{
   shapeType=Capsule
   length=1.9444
   radius=1
   shapeName=dyn_straps_r_a01_capsule
   center=<2.6596 -0.9064 0>
   rotation=<rot 90 -82.6474 -97.5251>
  }]
 dyn_straps_l_b03 [{shapeType=Capsule, length=1.9444, radius=1, shapeName=dyn_straps_r_a01_capsule, center=<1.3961 -0.5978 0>, rotation=<rot 90 0 0>}]
 dyn_straps_r_a02 [{
   shapeType=Capsule
   length=2.1944
   radius=1.25
   shapeName=dyn_straps_r_a01_capsule
   center=<1.2552 -0.4908 -0.2076>
   rotation=<rot 90 179.9603 179.9606>
  }]
 dyn_straps_r_a03 [{shapeType=Capsule, length=1.6944, radius=1.25, shapeName=dyn_straps_r_a01_capsule, center=<1.1067 -0.0466 0>, rotation=<rot 90 0 0>}]
 dyn_straps_r_b02 [{shapeType=Capsule, length=1.9444, radius=1, shapeName=dyn_straps_r_a01_capsule, center=<2.6596 -0.9064 0>, rotation=<rot 90 0 0>}]
 dyn_straps_r_b03 [{shapeType=Capsule, length=1.9444, radius=1.25, shapeName=dyn_straps_r_a01_capsule, center=<2.0828 -0.5978 -0.5874>, rotation=<rot 90 0 0>}]
 upperarm_twist_01_l [{shapeType=Capsule, length=2.7834, radius=4.9645, shapeName=upperarm_twist_01_r_capsule, center=<14.0773 0 0>, rotation=<rot 90 0 0>}]
 upperarm_twist_01_r [{
   shapeType=Capsule
   length=2.0334
   radius=5.2145
   shapeName=upperarm_twist_01_r_capsule
   center=<-13.1588 0.3663 0.5767>
   rotation=<rot 80 -179.9997 -179.9998>
  }]
}
constraints [
 {bone1Name=dyn_straps_r_a02, bone2Name=upperarm_twist_01_r, swing1LimitDegrees=45, swing2LimitDegrees=50, twistLimitDegrees=1}
 {bone1Name=dyn_straps_r_a03, bone2Name=dyn_straps_r_a02, swing1LimitDegrees=45, swing2LimitDegrees=50, twistLimitDegrees=1}
 {bone1Name=dyn_straps_r_b02, bone2Name=upperarm_twist_01_r, swing1LimitDegrees=15, swing2LimitDegrees=50, twistLimitDegrees=1}
 {bone1Name=dyn_straps_r_b03, bone2Name=dyn_straps_r_b02, swing1LimitDegrees=45, swing2LimitDegrees=50, twistLimitDegrees=1}
 {bone1Name=dyn_straps_l_a02, bone2Name=upperarm_twist_01_l, swing1LimitDegrees=45, swing2LimitDegrees=50, twistLimitDegrees=1}
 {bone1Name=dyn_straps_l_a03, bone2Name=dyn_straps_l_a02, swing1LimitDegrees=45, swing2LimitDegrees=50, twistLimitDegrees=1}
 {bone1Name=dyn_straps_l_b02, bone2Name=upperarm_twist_01_l, swing1LimitDegrees=15, swing2LimitDegrees=50, twistLimitDegrees=1}
 {bone1Name=dyn_straps_l_b03, bone2Name=dyn_straps_l_b02, swing1LimitDegrees=45, swing2LimitDegrees=50, twistLimitDegrees=1}
]
