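# Physics asset for the two-tailed scarf on the medium female Impulse Spring player body, for simulating its dynamic bones.
# Left and right seven-capsule chains hang from spine_05, each joint having moderate swing and very tight twist limits.
# Constraints with free rotation also link the first three left and right segment pairs across the chains.
# Pelvis, spine and thigh capsules, joined by angularly free constraints, form a simplified body around the scarf.
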PhysicsAsset /Game/Characters/Player/Female/Medium/Bodies/F_MED_Impulse_Spring/Meshes/F_MED_Impulse_Spring_Scarf_Physics
bodies {
 dyn_scarf_l_01 [{shapeType=Capsule, length=3.5936, radius=3, shapeName=dyn_scarf_l_01_capsule, center=<2.3561 -2 0>, rotation=<rot 90 0 0>}]
 dyn_scarf_l_02 [{shapeType=Capsule, length=3.5936, radius=3, shapeName=dyn_scarf_l_01_capsule, center=<2.3561 -2 0>, rotation=<rot 90 0 0>}]
 dyn_scarf_l_03 [{shapeType=Capsule, length=3.5936, radius=3, shapeName=dyn_scarf_l_01_capsule, center=<2.3561 -2 0>, rotation=<rot 90 0 0>}]
 dyn_scarf_l_04 [{shapeType=Capsule, length=3.3436, radius=3.5, shapeName=dyn_scarf_l_01_capsule, center=<2.3561 -2 0>}]
 dyn_scarf_l_05 [{shapeType=Capsule, length=3.3436, radius=3.5, shapeName=dyn_scarf_l_01_capsule, center=<2.3561 -2 0>}]
 dyn_scarf_l_06 [{shapeType=Capsule, length=3.3436, radius=3.5, shapeName=dyn_scarf_l_01_capsule, center=<2.3561 -2 0>}]
 dyn_scarf_l_07 [{shapeType=Capsule, length=3.3436, radius=3.5, shapeName=dyn_scarf_l_01_capsule, center=<2.3561 -2 0>}]
 dyn_scarf_r_01 [{shapeType=Capsule, length=3.5936, radius=3, shapeName=dyn_scarf_l_01_capsule, center=<2.3561 -2 0>, rotation=<rot 90 0 0>}]
 dyn_scarf_r_02 [{shapeType=Capsule, length=4.5936, radius=3.5, shapeName=dyn_scarf_l_01_capsule, center=<2.3561 -2 0>}]
 dyn_scarf_r_03 [{shapeType=Capsule, length=4.5936, radius=3.5, shapeName=dyn_scarf_l_01_capsule, center=<2.3561 -2 0>}]
 dyn_scarf_r_04 [{shapeType=Capsule, length=4.5936, radius=3.5, shapeName=dyn_scarf_l_01_capsule, center=<2.3561 -2 0>}]
 dyn_scarf_r_05 [{shapeType=Capsule, length=4.5936, radius=3.5, shapeName=dyn_scarf_l_01_capsule, center=<2.3561 -2 0>}]
 dyn_scarf_r_06 [{shapeType=Capsule, length=4.5936, radius=4.5, shapeName=dyn_scarf_l_01_capsule, center=<2.3561 -2 0>}]
 dyn_scarf_r_07 [{shapeType=Capsule, length=4.5936, radius=4.5, shapeName=dyn_scarf_l_01_capsule, center=<2.3561 -2 0>}]
 pelvis [{shapeType=Capsule, length=10.4683, radius=10.6329, shapeName=pelvis_capsule, center=<-4.2559 -2.0905 0>, rotation=<rot 0 -90 0>}]
 spine_02 [{shapeType=Capsule, length=18.1402, radius=7.7615, shapeName=spine_05_capsule, center=<-4.894 1.3293 0>}]
 spine_03 [{shapeType=Capsule, length=18.1402, radius=7.7615, shapeName=spine_05_capsule, center=<-4.894 1.3293 0>}]
 spine_05 [{shapeType=Capsule, length=16.1402, radius=9.7615, shapeName=spine_05_capsule, center=<-4.894 2.1876 0>, rotation=<rot 0 0 0.0562>}]
 thigh_l [{shapeType=Capsule, length=26.338, radius=8.0548, shapeName=thigh_l_capsule, center=<-23.1154 0 0>, rotation=<rot 90 0 0>}]
 thigh_r [{shapeType=Capsule, length=26.838, radius=7.5548, shapeName=thigh_l_capsule, center=<23.424 0 0>, rotation=<rot 90 179.5802 179.5802>}]
}
constraints [
 {
  bone1Name=thigh_l
  bone2Name=pelvis
  swing1LimitDegrees=45
  swing1Motion=Free
  swing2LimitDegrees=45
  swing2Motion=Free
  twistLimitDegrees=45
  twistMotion=Free
 }
 {
  bone1Name=thigh_r
  bone2Name=pelvis
  swing1LimitDegrees=45
  swing1Motion=Free
  swing2LimitDegrees=45
  swing2Motion=Free
  twistLimitDegrees=45
  twistMotion=Free
 }
 {
  bone1Name=spine_05
  bone2Name=spine_03
  swing1LimitDegrees=45
  swing1Motion=Free
  swing2LimitDegrees=45
  swing2Motion=Free
  twistLimitDegrees=45
  twistMotion=Free
 }
 {
  bone1Name=spine_03
  bone2Name=spine_02
  swing1LimitDegrees=45
  swing1Motion=Free
  swing2LimitDegrees=45
  swing2Motion=Free
  twistLimitDegrees=45
  twistMotion=Free
 }
 {bone1Name=dyn_scarf_l_01, bone2Name=spine_05, swing1LimitDegrees=35, swing2LimitDegrees=35, twistLimitDegrees=1}
 {bone1Name=dyn_scarf_l_07, bone2Name=dyn_scarf_l_06, swing1LimitDegrees=35, swing2LimitDegrees=35, twistLimitDegrees=1}
 {bone1Name=dyn_scarf_l_06, bone2Name=dyn_scarf_l_05, swing1LimitDegrees=35, swing2LimitDegrees=35, twistLimitDegrees=1}
 {bone1Name=dyn_scarf_l_05, bone2Name=dyn_scarf_l_04, swing1LimitDegrees=35, swing2LimitDegrees=35, twistLimitDegrees=1}
 {bone1Name=dyn_scarf_l_04, bone2Name=dyn_scarf_l_03, swing1LimitDegrees=35, swing2LimitDegrees=35, twistLimitDegrees=1}
 {bone1Name=dyn_scarf_l_03, bone2Name=dyn_scarf_l_02, swing1LimitDegrees=35, swing2LimitDegrees=35, twistLimitDegrees=1}
 {bone1Name=dyn_scarf_l_02, bone2Name=dyn_scarf_l_01, swing1LimitDegrees=35, swing2LimitDegrees=35, twistLimitDegrees=1}
 {bone1Name=dyn_scarf_r_01, bone2Name=spine_05, swing1LimitDegrees=35, swing2LimitDegrees=35, twistLimitDegrees=1}
 {bone1Name=dyn_scarf_r_02, bone2Name=dyn_scarf_r_01, swing1LimitDegrees=35, swing2LimitDegrees=35, twistLimitDegrees=1}
 {bone1Name=dyn_scarf_r_03, bone2Name=dyn_scarf_r_02, swing1LimitDegrees=35, swing2LimitDegrees=35, twistLimitDegrees=1}
 {bone1Name=dyn_scarf_r_04, bone2Name=dyn_scarf_r_03, swing1LimitDegrees=35, swing2LimitDegrees=35, twistLimitDegrees=1}
 {bone1Name=dyn_scarf_r_05, bone2Name=dyn_scarf_r_04, swing1LimitDegrees=35, swing2LimitDegrees=35, twistLimitDegrees=1}
 {bone1Name=dyn_scarf_r_06, bone2Name=dyn_scarf_r_05, swing1LimitDegrees=35, swing2LimitDegrees=35, twistLimitDegrees=1}
 {bone1Name=dyn_scarf_r_07, bone2Name=dyn_scarf_r_06, swing1LimitDegrees=35, swing2LimitDegrees=35, twistLimitDegrees=1}
 {
  bone1Name=spine_02
  bone2Name=pelvis
  swing1LimitDegrees=45
  swing1Motion=Free
  swing2LimitDegrees=45
  swing2Motion=Free
  twistLimitDegrees=45
  twistMotion=Free
 }
 {
  bone1Name=dyn_scarf_l_01
  bone2Name=dyn_scarf_r_01
  swing1LimitDegrees=45
  swing1Motion=Free
  swing2LimitDegrees=45
  swing2Motion=Free
  twistLimitDegrees=45
  twistMotion=Free
 }
 {
  bone1Name=dyn_scarf_l_02
  bone2Name=dyn_scarf_r_02
  swing1LimitDegrees=45
  swing1Motion=Free
  swing2LimitDegrees=45
  swing2Motion=Free
  twistLimitDegrees=45
  twistMotion=Free
 }
 {
  bone1Name=dyn_scarf_l_03
  bone2Name=dyn_scarf_r_03
  swing1LimitDegrees=45
  swing1Motion=Free
  swing2LimitDegrees=45
  swing2Motion=Free
  twistLimitDegrees=45
  twistMotion=Free
 }
]
