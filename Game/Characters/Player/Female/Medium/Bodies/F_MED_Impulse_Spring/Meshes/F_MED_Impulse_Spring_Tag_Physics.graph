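# Physics asset for the dangling tag on the medium female player's Impulse Spring body mesh.
# A three-segment capsule chain hangs from a master tag body, with limited swing and very tight twist at each joint.
# The master links to spine_05 with free angular motion, as does spine_05 to spine_04, whose larger capsules stand in for the upper torso.
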PhysicsAsset /Game/Characters/Player/Female/Medium/Bodies/F_MED_Impulse_Spring/Meshes/F_MED_Impulse_Spring_Tag_Physics
bodies {
 dyn_tag_01 [{shapeType=Capsule, length=3.3436, radius=3.25, shapeName=dyn_scarf_l_01_capsule, center=<1.4831 0 1.9529>, rotation=<rot 0 90 90.0006>}]
 dyn_tag_02 [{shapeType=Capsule, length=4.3436, radius=3.75, shapeName=dyn_scarf_l_01_capsule, center=<1.5691 0.0537 1.9447>, rotation=<rot 90 180 -180>}]
 dyn_tag_03 [{shapeType=Capsule, length=4.3436, radius=3.75, shapeName=dyn_scarf_l_01_capsule, center=<2.3561 0 2.2264>, rotation=<rot 0 90 90.0006>}]
 dyn_tag_MSTR [{
   shapeType=Capsule
   length=1.4734
   radius=1.1199
   shapeName=dyn_tag_MSTR_capsule
   center=<0 -0.1283 -0.4533>
   rotation=<rot -32.9404 -90 -0.0019>
  }]
 spine_04 [{shapeType=Capsule, length=17.1402, radius=8.7615, shapeName=spine_05_capsule, center=<-4.894 2.1876 0>}]
 spine_05 [{shapeType=Capsule, length=17.1402, radius=8.7615, shapeName=spine_05_capsule, center=<-5.4532 2.7542 0>}]
}
constraints [
 {bone1Name=dyn_tag_03, bone2Name=dyn_tag_02, swing1LimitDegrees=35, swing2LimitDegrees=35, twistLimitDegrees=1}
 {bone1Name=dyn_tag_02, bone2Name=dyn_tag_01, swing1LimitDegrees=35, swing2LimitDegrees=35, twistLimitDegrees=1}
 {bone1Name=dyn_tag_01, bone2Name=dyn_tag_MSTR, swing1LimitDegrees=35, swing2LimitDegrees=35, twistLimitDegrees=1}
 {
  bone1Name=spine_05
  bone2Name=spine_04
  swing1LimitDegrees=45
  swing1Motion=Free
  swing2LimitDegrees=45
  swing2Motion=Free
  twistLimitDegrees=45
  twistMotion=Free
 }
 {
  bone1Name=dyn_tag_MSTR
  bone2Name=spine_05
  swing1LimitDegrees=45
  swing1Motion=Free
  swing2LimitDegrees=45
  swing2Motion=Free
  twistLimitDegrees=45
  twistMotion=Free
 }
]
