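# Physics asset for the belt of the female medium Impulse Spring player body, setting up jointed belt pieces around the hips.
# Left and right three-bone dyn_belt chains of small capsules attach to the pelvis, each joint carrying angular limits, with a tighter first swing at the roots.
# Larger pelvis and thigh capsules supply body shapes around the belt, with the thighs joined to the pelvis by free-rotation constraints.
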PhysicsAsset /Game/Characters/Player/Female/Medium/Bodies/F_MED_Impulse_Spring/Meshes/F_MED_Impulse_Spring_belt_Physics
bodies {
 dyn_belt_l_01 [{shapeType=Capsule, length=2.4178, radius=3, shapeName=dyn_belt_r_01_capsule, center=<2.4425 -2 0>, rotation=<rot 90 112.5547 111.9408>}]
 dyn_belt_l_02 [{shapeType=Capsule, length=2.4178, radius=3, shapeName=dyn_belt_r_01_capsule, center=<2.4425 -2 0>, rotation=<rot 90 0 0>}]
 dyn_belt_l_03 [{shapeType=Capsule, length=2.4178, radius=3, shapeName=dyn_belt_r_01_capsule, center=<2.4425 -2 0>, rotation=<rot 90 0 0>}]
 dyn_belt_r_01 [{shapeType=Capsule, length=2.4178, radius=3, shapeName=dyn_belt_r_01_capsule, center=<2.4425 -2 0>, rotation=<rot 90 0 0>}]
 dyn_belt_r_02 [{shapeType=Capsule, length=2.4178, radius=3, shapeName=dyn_belt_r_01_capsule, center=<2.4425 -2 0>, rotation=<rot 90 0 0>}]
 dyn_belt_r_03 [{shapeType=Capsule, length=2.4178, radius=3, shapeName=dyn_belt_r_01_capsule, center=<2.4425 -2 0>, rotation=<rot 90 0 0>}]
 pelvis [{shapeType=Capsule, length=10.4683, radius=10.6329, shapeName=pelvis_capsule, center=<-4.2559 -2.0905 0>, rotation=<rot 0 -90 0>}]
 thigh_l [{shapeType=Capsule, length=26.338, radius=8.0548, shapeName=thigh_l_capsule, center=<-23.1154 0 0>, rotation=<rot 90 0 0>}]
 thigh_r [{shapeType=Capsule, length=26.838, radius=7.5548, shapeName=thigh_l_capsule, center=<23.424 0 0>, rotation=<rot 90 0 0>}]
}
constraints [
 {
  bone1Name=thigh_l
  bone2Name=pelvis
  swing1LimitDegrees=45
  swing1Motion=Free
  swing2LimitDegrees=45
  swing2Motion=Free
  twistLimitDegrees=45
  twistMotion=Free
 }
 {
  bone1Name=thigh_r
  bone2Name=pelvis
  swing1LimitDegrees=45
  swing1Motion=Free
  swing2LimitDegrees=45
  swing2Motion=Free
  twistLimitDegrees=45
  twistMotion=Free
 }
 {bone1Name=dyn_belt_r_01, bone2Name=pelvis, swing1LimitDegrees=25, swing2LimitDegrees=45, twistLimitDegrees=45}
 {bone1Name=dyn_belt_r_02, bone2Name=dyn_belt_r_01, swing1LimitDegrees=45, swing2LimitDegrees=45, twistLimitDegrees=45}
 {bone1Name=dyn_belt_r_03, bone2Name=dyn_belt_r_02, swing1LimitDegrees=45, swing2LimitDegrees=45, twistLimitDegrees=45}
 {bone1Name=dyn_belt_l_03, bone2Name=dyn_belt_l_02, swing1LimitDegrees=45, swing2LimitDegrees=45, twistLimitDegrees=45}
 {bone1Name=dyn_belt_l_02, bone2Name=dyn_belt_l_01, swing1LimitDegrees=45, swing2LimitDegrees=45, twistLimitDegrees=45}
 {bone1Name=dyn_belt_l_01, bone2Name=pelvis, swing1LimitDegrees=25, swing2LimitDegrees=45, twistLimitDegrees=45}
]
