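# Animation Blueprint for the F_MED_Impulse_Spring player body, adding secondary motion to its clothing and accessories.
# It copies the attached parent's pose, applies a Control Rig while skydiving, then chains bone constraints, AnimDynamics and RigidBody simulations for straps, hood, arms, scarf, belt, rings and zipper.
# The tag RigidBody has zero alpha.
# A customization meta tag sets scarf/tag flags, which the AnimGraph never visibly reads.
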AnimBlueprint /Game/Characters/Player/Female/Medium/Bodies/F_MED_Impulse_Spring/Meshes/F_MED_Impulse_Spring_AnimBP
target_skeleton=&/F_MED_Impulse_Spring_Skeleton
graphs [
 AnimGraph {nodes [
   "Constraint - dyn_straps_l_b01 (PPP)"!(ComponentPose, Alpha=1, ConstraintWeights_0=0.566, ConstraintWeights_1:float=0.224, ConstraintWeights_2:float=0.21) -> (Pose) {
    alphaBoolBlend {blendOption=Linear}
    boneToModify {boneName=dyn_straps_l_b01}
    constraintSetup [{targetBone {boneName=upperarm_twist_01_l}}, {targetBone {boneName=upperarm_twist_02_l}}, {targetBone {boneName=upperarm_l}}]
    constraintWeights [1,1,1]
   }
   n1 = Control Rig!(Source=Copy Pose From Mesh() {bCopyCurves=true, bUseAttachedParent=true}, bAlphaBoolEnabled:bool=Get Skydive() -> (Skydive:bool)) -> (Pose) {
    alphaInputType=Bool
    alphaBoolBlend {blendOption=Linear}
    controlRigAssetReference {blueprintRigClass=&/F_MED_Impulse_Spring_CtrlRig.F_MED_Impulse_Spring_CtrlRig_C}
   }
   n2 = "Constraint - dyn_straps_r_b01 (PPP)"!(
    ComponentPose=Local To Component(LocalPose=@n1)
    Alpha=1
    ConstraintWeights_0=0.567
    ConstraintWeights_1:float=0.223
    ConstraintWeights_2:float=0.21
   ) -> (Pose) {
    alphaBoolBlend {blendOption=Linear}
    boneToModify {boneName=dyn_straps_r_b01}
    constraintSetup [{targetBone {boneName=upperarm_twist_01_r}}, {targetBone {boneName=upperarm_twist_02_r}}, {targetBone {boneName=upperarm_r}}]
    constraintWeights [1,1,1]
   }
   n3 = "Constraint - dyn_straps_l_b01 (PPP)"(ComponentPose=@n2, Alpha=1, ConstraintWeights_0=0.567, ConstraintWeights_1=0.223, ConstraintWeights_2=0.21) {
    alphaBoolBlend {blendOption=Linear}
    boneToModify {boneName=dyn_straps_l_b01}
    constraintSetup [{targetBone {boneName=upperarm_twist_01_l}}, {targetBone {boneName=upperarm_twist_02_l}}, {targetBone {boneName=upperarm_l}}]
    constraintWeights [1,1,1]
   }
   n4 = "Constraint - dyn_tag_MSTR (PP)"!(ComponentPose=@n3, Alpha=1, ConstraintWeights_0=0.7161, ConstraintWeights_1:float=0.2839) -> (Pose) {
    alphaBoolBlend {blendOption=Linear}
    boneToModify {boneName=dyn_tag_MSTR}
    constraintSetup [{targetBone {boneName=spine_05}}, {targetBone {boneName=spine_04}}]
    constraintWeights [1,1]
   }
   n5 = Anim Dynamics - Bone: dyn_hood1(ComponentPose=@n4, Alpha=1, ExternalForce="0,0,0") {
    angularDampingOverride=0.8
    angularSpringConstant=500
    bAngularSpring=true
    bGravityOverrideInSimSpace=true
    bUseGravityOverride=true
    simulationSpace=BoneRelative
    alphaBoolBlend {blendOption=Linear}
    boundBone {boneName=dyn_hood1}
    componentAppliedLinearAccClamp=<5000 5000 5000>
    componentLinearAccScale=<0.8 0.8 0.8>
    componentLinearVelScale=<0.4 0.4 0.4>
    gravityOverride {z=-980}
    physicsBodyDefinitions [{
      sphereCollisionRadius=0
      boundBone {boneName=dyn_hood1}
      boxExtents=<0.5 20 20>
      constraintSetup {angularLimitsMax=<0 0 0>, angularLimitsMin=<-30 -80 -30>, angularTarget=<1 0 0>}
      localJointOffset=<1 0 0>
     }]
    relativeSpaceBone {boneName=root}
   }
   n6 = RigidBody(ComponentPose=@n5) {
    bClampLinearTranslationLimitToRefPose=true
    bOverrideWorldGravity=true
    overridePhysicsAsset=&/F_MED_Impulse_Spring_Arms_Physics
    simulationSpace=BaseBoneSpace
    alphaBoolBlend {blendOption=Linear}
    baseBoneRef {boneName=spine_05}
    componentAppliedLinearAccClamp=<3000 3000 3000>
    componentLinearAccScale=<0.45 0.2 0.2>
    componentLinearVelScale=<0.2 0.1 0.1>
    overrideWorldGravity {z=-980}
    simSpaceSettings {
     maxAngularAcceleration=1000
     maxAngularVelocity=3
     maxLinearAcceleration=700
     maxLinearVelocity=500
     velocityScaleZ=0
     externalLinearDragV=<0.2 0.2>
    }
   }
   n7 = RigidBody(ComponentPose=@n6) {
    bClampLinearTranslationLimitToRefPose=true
    bOverrideWorldGravity=true
    overridePhysicsAsset=&/F_MED_Impulse_Spring_Scarf_Physics
    simulationSpace=BaseBoneSpace
    alphaBoolBlend {blendOption=Linear}
    baseBoneRef {boneName=pelvis}
    componentAppliedLinearAccClamp=<3000 3000 3000>
    componentLinearAccScale=<0.45 0.2 0.2>
    componentLinearVelScale=<0.2 0.1 0.1>
    overrideWorldGravity {z=-980}
    simSpaceSettings {
     maxAngularAcceleration=1000
     maxAngularVelocity=3
     maxLinearAcceleration=700
     maxLinearVelocity=500
     velocityScaleZ=0
     worldAlpha=1
     externalLinearDragV=<0.2 0.2>
    }
   }
   n8 = RigidBody(ComponentPose=@n7, Alpha=0) {
    bClampLinearTranslationLimitToRefPose=true
    bOverrideWorldGravity=true
    overridePhysicsAsset=&/F_MED_Impulse_Spring_Tag_Physics
    simulationSpace=BaseBoneSpace
    alphaBoolBlend {blendOption=Linear}
    baseBoneRef {boneName=spine_05}
    componentAppliedLinearAccClamp=<3000 3000 3000>
    componentLinearAccScale=<0.2 0.2 0.45>
    componentLinearVelScale=<0.2 0.2 0.25>
    overrideWorldGravity {z=-980}
    simSpaceSettings {
     maxAngularAcceleration=1000
     maxAngularVelocity=3
     maxLinearAcceleration=700
     maxLinearVelocity=500
     velocityScaleZ=0
     worldAlpha=1
     externalLinearDragV=<0.2 0.2>
    }
   }
   n9 = RigidBody(ComponentPose=@n8) {
    bClampLinearTranslationLimitToRefPose=true
    bOverrideWorldGravity=true
    overridePhysicsAsset=&/F_MED_Impulse_Spring_belt_Physics
    simulationSpace=BaseBoneSpace
    alphaBoolBlend {blendOption=Linear}
    baseBoneRef {boneName=pelvis}
    componentAppliedLinearAccClamp=<3000 3000 3000>
    componentLinearAccScale=<0.45 0.2 0.2>
    componentLinearVelScale=<0.2 0.1 0.1>
    overrideWorldGravity {z=-980}
    simSpaceSettings {
     maxAngularAcceleration=1000
     maxAngularVelocity=3
     maxLinearAcceleration=700
     maxLinearVelocity=500
     velocityScaleZ=0
     worldAlpha=1
     externalLinearDragV=<0.2 0.2>
    }
   }
   n10 = "Constraint - dyn_ring_r (PPP)"!(ComponentPose=@n9, Alpha=1, ConstraintWeights_0=0.694, ConstraintWeights_1:float=0.251, ConstraintWeights_2:float=0.055) -> (Pose) {
    alphaBoolBlend {blendOption=Linear}
    boneToModify {boneName=dyn_ring_r}
    constraintSetup [{targetBone {boneName=spine_04}}, {targetBone {boneName=spine_05}}, {targetBone {boneName=clavicle_r}}]
    constraintWeights [1,1,1]
   }
   n11 = Anim Dynamics - Bone: dyn_ring_r(ComponentPose=@n10, Alpha=1, ExternalForce="0,0,0") {
    angularDampingOverride=0.3
    angularSpringConstant=800
    bAngularSpring=true
    bGravityOverrideInSimSpace=true
    bUseGravityOverride=true
    simulationSpace=BoneRelative
    alphaBoolBlend {blendOption=Linear}
    boundBone {boneName=dyn_ring_r}
    componentAppliedLinearAccClamp=<5000 5000 5000>
    componentLinearAccScale=<0.5 0.5 0.7>
    componentLinearVelScale=<0.25 0.25 0.25>
    gravityOverride {z=-980}
    physicsBodyDefinitions [{
      sphereCollisionRadius=0
      boundBone {boneName=dyn_ring_r}
      boxExtents=<2 30 25>
      constraintSetup {angularLimitsMax=<30 0 10>, angularLimitsMin=<-30 0 -60>, angularTarget=<0 20 0>}
      localJointOffset=<4 0 0>
     }]
    relativeSpaceBone {boneName=root}
   }
   n12 = "Constraint - dyn_ring_l (PPP)"!(ComponentPose=@n11, Alpha=1, ConstraintWeights_0=0.694, ConstraintWeights_1:float=0.251, ConstraintWeights_2:float=0.055) -> (Pose) {
    alphaBoolBlend {blendOption=Linear}
    boneToModify {boneName=dyn_ring_l}
    constraintSetup [{targetBone {boneName=spine_04}}, {targetBone {boneName=spine_05}}, {targetBone {boneName=clavicle_l}}]
    constraintWeights [1,1,1]
   }
   n13 = Anim Dynamics - Bone: dyn_ring_l(ComponentPose=@n12, Alpha=1, ExternalForce="0,0,0") {
    angularDampingOverride=0.3
    angularSpringConstant=800
    bAngularSpring=true
    bGravityOverrideInSimSpace=true
    bUseGravityOverride=true
    simulationSpace=BoneRelative
    alphaBoolBlend {blendOption=Linear}
    boundBone {boneName=dyn_ring_l}
    componentAppliedLinearAccClamp=<5000 5000 5000>
    componentLinearAccScale=<0.5 0.5 0.7>
    componentLinearVelScale=<0.25 0.25 0.25>
    gravityOverride {z=-980}
    physicsBodyDefinitions [{
      sphereCollisionRadius=0
      boundBone {boneName=dyn_ring_l}
      boxExtents=<2 30 25>
      constraintSetup {angularLimitsMax=<30 0 10>, angularLimitsMin=<-30 0 -60>, angularTarget=<0 20 0>}
      localJointOffset=<4 0 0>
     }]
    relativeSpaceBone {boneName=root}
   }
   n14 = "Constraint - dyn_zipper_MSTR (PPP)"!(ComponentPose=@n13, Alpha=1, ConstraintWeights_0=0.642, ConstraintWeights_1:float=0.325, ConstraintWeights_2:float=0.033) -> (Pose) {
    alphaBoolBlend {blendOption=Linear}
    boneToModify {boneName=dyn_zipper_MSTR}
    constraintSetup [{targetBone {boneName=thigh_r}}, {targetBone {boneName=thigh_twist_01_r}}, {targetBone {boneName=pelvis}}]
    constraintWeights [1,1,1]
   }
   n15 = Anim Dynamics - Bone: dyn_zipper_01(ComponentPose=@n14, Alpha=1, ExternalForce="0,0,0") {
    angularDampingOverride=3
    angularSpringConstant=1500
    bAngularSpring=true
    bGravityOverrideInSimSpace=true
    bUseGravityOverride=true
    simulationSpace=BoneRelative
    alphaBoolBlend {blendOption=Linear}
    boundBone {boneName=dyn_zipper_01}
    componentAppliedLinearAccClamp=<5000 5000 5000>
    componentLinearAccScale=<0.2 0.2 0.55>
    componentLinearVelScale=<0.2 0.2 0.3>
    gravityOverride {z=-1000}
    physicsBodyDefinitions [{
      sphereCollisionRadius=0
      boundBone {boneName=dyn_zipper_01}
      boxExtents=<2 20 20>
      constraintSetup {angularLimitsMax=<0 0 0>, angularLimitsMin=<-0.1 0 -80>, angularTarget=<1 0 0>}
      localJointOffset=<5 0 0>
     }]
    relativeSpaceBone {boneName=root}
   }
   Output Pose(Result=Component To Local(ComponentPose=@n15)) {layerGroup=None}
  ]}
 EventGraph {nodes [
   Event BlueprintUpdateAnimation()
   > Set Skydive(Skydive:bool=Get bIsSkyDiving()) -> (Output_Get:bool)
   Event OnCharacterCustomizationLoaded()
   > n1 = Branch(Condition=CharacterItemDefinitionHasAnyMetaTag(Tags=Make Array("[0]"=(TagName="Cosmetics.MetaTags.Impulse.ScarfOn"))))
   > Set TagActive(TagActive:bool=false) -> (Output_Get:bool)
   > Set ScrarfActive(ScrarfActive:bool=true) -> (Output_Get:bool)
   Set TagActive(execute=@n1.else, TagActive=true)
   > Set ScrarfActive(ScrarfActive=false)
   Get bIsSkydiveDiveMode()
  ]}
]
variables [
 ScrarfActive {type=boolean}
 Skydive {type=boolean}
 TagActive {type=boolean}
]
